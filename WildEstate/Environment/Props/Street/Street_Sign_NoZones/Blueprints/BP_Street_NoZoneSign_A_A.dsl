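; Street no-zone sign prop that reacts to a Season 38 story quest through gameplay-tagged side effects.
; On begin play it stops its prop audio component.
; The Sign tag turns on the mesh's custom depth, plays the audio, then enables query and physics collision after a short delay.
; The SignOff tag turns off custom depth, makes the sign damageable and stops the audio.
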(event EventBeginPlay
  (|CallParentFunction)
  (Class|FortPropAudioComponent|StopAudio :self (Variables|Default|GetFortPropAudioComponent) :Damage 0.0 :Momentum "0, 0, 0"))

(event Collision|EventActorBeginOverlap
  (params (OtherActor :type Actor%20Object%20Reference))
  (|CallParentFunction OtherActor) ; This node is disabled and will not be called.\nDrag off pins to build functionality.
)

(event EventTick
  (params (DeltaSeconds :type Float%20%28single-precision%29))
  (|CallParentFunction DeltaSeconds) ; This node is disabled and will not be called.\nDrag off pins to build functionality.
)

(event SideEffects|EventBP_ApplySideEffect
  (params (ReactionSource :type Object%20Reference) (TargetObject :type Object%20Reference) (SideEffectTriggerTag :type Gameplay%20Tag%20Structure%20%28by%20ref%29) (SideEffectTriggerType :type EBlueprintableSideEffectTriggerType%20Enum%20%28by%20ref%29 :default "TriggeredAsTarget"))
  (switch Utilities|FlowControl|Switch|SwitchonGameplayTag SideEffectTriggerTag
    (:Athena.Quests.S38.Event.Story.P01.Sign
      (Rendering|SetRenderCustomDepth (Variables|Debug|GetStaticMeshComponent) true)
      (Class|FortPropAudioComponent|PlayAudio (Variables|Default|GetFortPropAudioComponent))
      (Utilities|FlowControl|Delay 1.0)
      (Collision|SetCollisionEnabled (Variables|Debug|GetStaticMeshComponent) "QueryAndPhysics"))
    (:Athena.Quests.S38.Event.Story.P01.SignOff
      (Rendering|SetRenderCustomDepth (Variables|Debug|GetStaticMeshComponent) false)
      (Variables|Actor|SetbCanBeDamaged true)
      (Class|FortPropAudioComponent|StopAudio :self (Variables|Default|GetFortPropAudioComponent) :Damage 0.0 :Momentum "0, 0, 0"))))

(fn ConstructionScript
  (|CallParentFunction))
